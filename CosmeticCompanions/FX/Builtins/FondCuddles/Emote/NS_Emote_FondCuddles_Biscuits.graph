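# Niagara effect for the FondCuddles companion "Biscuits" emote, showing mesh cookies that float up and pop amid pink flour puffs.
# Cookies spawn from offsets on either side, bounce in scale and drift upward on curl noise.
# Paw puffs fire in timed bursts at side offsets.
# Cookies write deaths to a system array; ReadDeath and system variables drive pop puffs and sparkles at those positions, with puffs tinted by time of day.
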NiagaraSystem /CosmeticCompanions/FX/Builtins/FondCuddles/Emote/NS_Emote_FondCuddles_Biscuits
dependencies {
 usedDataInterfaces [
  &NiagaraDataInterfaceArrayColor
  &NiagaraDataInterfaceArrayFloat3
  &NiagaraDataInterfaceVectorCurve
  &NiagaraDataInterfaceCurve
  &NiagaraDataInterfaceColorCurve
  &NiagaraDataInterfaceSkeletalMesh
  &NiagaraDataInterfaceArrayPosition
 ]
 usedDynamicInputs [&SelectLinearColorFromArray, &MaskVector]
 usedModules [
  &self:SystemSpawnScript.SystemScriptSource.SystemScriptGraph.NiagaraNodeAssignment_0.SetVariables_#1
  &self:SystemSpawnScript.SystemScriptSource.SystemScriptGraph.NiagaraNodeAssignment_3.SetVariables_#2
  &self:SystemSpawnScript.SystemScriptSource.SystemScriptGraph.NiagaraNodeAssignment_4.SetVariables_#3
  &self:SystemSpawnScript.SystemScriptSource.SystemScriptGraph.NiagaraNodeAssignment_1.SetVariables_#4
  &SystemState
  &self:ReadDeath
  &self:SystemSpawnScript.SystemScriptSource.SystemScriptGraph.NiagaraNodeAssignment_2.SetVariables_#5
 ]
}
emitters {
 Cookies {
  emitterUpdateScript [
   &EmitterState(Life Cycle Mode=::NewEnumerator1, Loop Behavior=::NewEnumerator1, Loop Duration=3.4, UseLoopDelay=false)
   &SpawnRate(SpawnRate=1.9, Use Spawn Probability=false)
  ]
  particleSpawnScript [
   &PartitionParticles()
   &InitializeParticle(
    Color Mode=::NewEnumerator0
    Lifetime Max=2.2
    Lifetime Min=1.7
    Lifetime Mode=::NewEnumerator1
    Mesh Scale Mode=::NewEnumerator3
    Sprite Size Mode=::NewEnumerator3
    Uniform Sprite Size=3
    UsePositionOffset=false
    Write Material Random=true
   )
   &AddVelocity(Velocity=&RandomRangeVector(Minimum=<0 20 60>, Maximum=<0 45 90>)) {module=AddVelocity001}
   &SystemLocation(
    MaskSpawnGroup=false
    Offset=&SelectVectorFromArray(
     Direct Array Index=$Particles.PartitionParticles.Partition
     Direct Array Mode=::NewEnumerator1
     Vector Selection Array=DataInterface() {propertyValues {FloatData [<10 0 0>, <-10 0 0>, <10 0 0>, <-10 0 0>, <10 0 0>, <-10 0 0>]}}
    )
   )
   &InitialMeshOrientation(Mesh Orientation Mode=::NewEnumerator7, Use Rotation Vector=true)
  ]
  particleUpdateScript [
   &ParticleState()
   &CurlNoiseForce(Noise Strength=&RandomRangeFloat(Minimum=-100, Maximum=100), Pan Noise=false)
   &GravityForce(Gravity=&RandomRangeVector(
     Minimum=&VectorFromCurve(VectorCurve=DataInterface() {propertyValues {
        XCurve {keys [<at 0 0>,<at 1 0>]}
        YCurve {keys [<at 0 0>,<at 1 0>]}
        ZCurve {keys [
          {interpMode=RCIM_Cubic, tangentMode=RCTM_User, time=0, value=10}
          {arriveTangent=88.4209, interpMode=RCIM_Cubic, leaveTangent=88.4209, tangentMode=RCTM_User, time=0.75, value=35}
         ]}
       }})
     Maximum=&VectorFromCurve(VectorCurve=DataInterface() {propertyValues {
        XCurve {keys [<at 0 0>,<at 1 0>]}
        YCurve {keys [<at 0 0>,<at 1 0>]}
        ZCurve {keys [
          {interpMode=RCIM_Cubic, tangentMode=RCTM_User, time=0, value=10}
          {arriveTangent=133.5137, interpMode=RCIM_Cubic, leaveTangent=133.5137, tangentMode=RCTM_User, time=0.75, value=70}
         ]}
       }})
    ))
   &DynamicMaterialParameters(
    Index 0 Param 1=&FloatFromCurve(FloatCurve=DataInterface() {propertyValues {Curve {keys [
         {arriveTangent=0.1706, interpMode=RCIM_Cubic, leaveTangent=0.1706, leaveTangentWeight=0.3081, tangentMode=RCTM_User, time=0, value=0}
         {arriveTangentWeight=0.3104, interpMode=RCIM_Cubic, leaveTangentWeight=0.0345, tangentMode=RCTM_User, time=0.9, value=0.2}
         {arriveTangent=3.2005, interpMode=RCIM_Cubic, leaveTangent=3.2005, tangentMode=RCTM_User, time=1, value=0.5}
        ]}}})
    Index 0 Param 2=0.15
    Module.Index0Float0Write=false
    Module.Index0Float1Write=true
    Module.Index0Float2Write=true
    Module.Index0Float3Write=true
   ) {enabled=false}
   &ScaleColor(
    Scale Alpha=&FloatFromCurve(
     FloatCurve=DataInterface() {propertyValues {Curve {
        defaultValue=0.3
        keys [
         {tangentMode=RCTM_User, time=0, value=1}
         {arriveTangent=-0.0405, interpMode=RCIM_Cubic, leaveTangent=-0.0405, tangentMode=RCTM_User, time=0.9, value=1}
         {arriveTangentWeight=0.1685, interpMode=RCIM_Cubic, tangentMode=RCTM_User, time=1, value=0.5}
        ]
       }}}
     Scale Curve=&FNTimeOfDayScalar(Night Scale=1, Morning Scale=1, Evening Scale=1)
    )
    Scale RGB=&MakeVectorFromLinearColorRGB(Color=&FNTimeOfDayColor(
      Night Color=<rgba 0.3609 0.4715 0.5469 1>
      Morning Color=<rgba 0.651 0.4517 0.4205 1>
      Day Color=<rgba 1 1 1 1>
      Evening Color=<rgba 0.651 0.4517 0.4205 1>
     ))
    ScaleA=true
    ScaleRGB=true
   ) {enabled=false}
   &ParticleState() {module=ParticleState001}
   &ScaleMeshSize(Scale Factor=&VectorFromCurve(
     VectorCurve=DataInterface() {propertyValues {
       XCurve {keys [<at 0 0>,<at 0.0687 1>,<at 0.95 1>,<at 1 1.3>]}
       YCurve {keys [<at 0 0>,<at 0.0687 1>,<at 0.95 1>,<at 1 1.3>]}
       ZCurve {keys [
         <at 0 0 RCIM_Cubic>
         {arriveTangent=15.6387, interpMode=RCIM_Cubic, leaveTangent=15.6387, time=0.0344, value=1}
         <at 0.1279 2 RCIM_Cubic>
         {arriveTangent=-6.2971, interpMode=RCIM_Cubic, leaveTangent=-6.2971, time=0.2443, value=1}
         {arriveTangent=0.6764, interpMode=RCIM_Cubic, leaveTangent=0.6764, time=0.3264, value=0.75}
         {arriveTangent=1.2815, interpMode=RCIM_Cubic, leaveTangent=1.2815, time=0.4291, value=1.125}
         {arriveTangentWeight=0.0318, interpMode=RCIM_Cubic, leaveTangentWeight=0.1343, tangentMode=RCTM_User, time=0.5215, value=1}
         {arriveTangentWeight=0.3013, interpMode=RCIM_Cubic, leaveTangentWeight=0.0682, tangentMode=RCTM_User, time=0.95, value=1}
         <at 1 1.3>
        ]}
      }}
     Scale Curve=&RandomRangeVector(Minimum=<1 1 1>, Maximum=<1 1 1>)
    ))
   &UpdateMeshOrientation(Rotation Vector=&RandomRangeVector(Minimum=<0 0 -0.18>, Maximum=<0 0 0.18>))
   &Drag(
    Drag=&FloatFromCurve(FloatCurve=DataInterface() {propertyValues {Curve {keys [<at 0 1 RCIM_Cubic>,<at 1 3 RCIM_Cubic>]}}})
    Use Linear Drag=true
    Use Rotational Drag=true
   )
   &SolveForcesAndVelocity(Clamp Velocity=false, Limit Acceleration=false)
   &self:WriteDeath(Position Array:NiagaraDataInterfaceArrayPosition=$System.DeathArray) {module=WriteDeath}
   &self:WriteVectorArray(Index:NiagaraInt32=$Particles.PartitionParticles.Partition, Vector Array:NiagaraDataInterfaceArrayFloat3=$System.PartPos) {enabled=false, module=WriteVectorArray}
  ]
  rendererClasses [&NiagaraMeshRendererProperties]
  renderers [{rendererClass=&NiagaraMeshRendererProperties, rendererIndex=0}]
 }
 FlourPuffs_Paws {
  emitterUpdateScript [
   &EmitterState()
   &SpawnBurst_Instantaneous(Spawn Count=4, Spawn Time=3.1, Use Spawn Probability=false, UseLoopCountLimit=false) {module=SpawnBurst_Instantaneous005}
   &SpawnBurst_Instantaneous(Spawn Count=4, Spawn Time=2.66, Use Spawn Probability=false, UseLoopCountLimit=false) {module=SpawnBurst_Instantaneous004}
   &SpawnBurst_Instantaneous(Spawn Count=4, Spawn Time=2.1, Use Spawn Probability=false, UseLoopCountLimit=false) {module=SpawnBurst_Instantaneous003}
   &SpawnBurst_Instantaneous(Spawn Count=4, Spawn Time=1.6, Use Spawn Probability=false, UseLoopCountLimit=false) {module=SpawnBurst_Instantaneous002}
   &SpawnBurst_Instantaneous(Spawn Count=4, Spawn Time=1.05, Use Spawn Probability=false, UseLoopCountLimit=false) {module=SpawnBurst_Instantaneous001}
   &SpawnBurst_Instantaneous(Spawn Count=4, Spawn Time=0.6, Use Spawn Probability=false, UseLoopCountLimit=false)
  ]
  particleSpawnScript [
   &PartitionParticles(
    Particles Per Partition=4
    Partition Count Mode:ENiagara_InfiniteLoopDuration=::NewEnumerator1
    Partition Mode:ENiagara_PartitionMode_v2=::NewEnumerator1
   )
   &InitializeParticle(
    Color Maximum=<rgba 1 0.723 0.6264 0.55>
    Color Minimum=<rgba 1 0.3703 0.7133 0.73>
    Color Mode=::NewEnumerator2
    Lifetime Max=1.1
    Lifetime Min=0.9
    Lifetime Mode=::NewEnumerator1
    Sprite Rotation Mode=::NewEnumerator2
    Sprite Size Mode=::NewEnumerator1
    Uniform Sprite Size Max=23
    Uniform Sprite Size Min=12
    UsePositionOffset=false
    Write Material Random=true
   )
   &AddVelocity(
    Cone Angle=112
    Cone Axis=<0 0 1>
    Module.UseConeFalloff=false
    Module.UseVelDistribution=false
    Override Seed=false
    Velocity Mode=::NewEnumerator2
    Velocity Speed=&RandomRangeFloat(Minimum=35, Maximum=65)
   ) {module=AddVelocity001}
   &CameraOffset(Camera Offset Amount=10)
   &ShapeLocation(
    Handle Radius=0
    Large Radius=3
    MaskSpawnGroup=false
    Non Uniform Scale=<1 1 0>
    Offset Mode=::NewEnumerator0
    Override Seed=false
    Shape Primitive=::NewEnumerator3
    UseOwnerScale=true
   )
   &SystemLocation(
    MaskSpawnGroup=false
    Offset=&SelectVectorFromArray(
     Direct Array Index=$Particles.PartitionParticles.Partition
     Direct Array Mode=::NewEnumerator1
     Vector Selection Array=DataInterface() {propertyValues {FloatData [<7 0 -5>, <-7 0 -5>, <7 0 -5>, <-7 0 -5>, <7 0 -5>, <-7 0 -5>]}}
    )
   )
  ]
  particleUpdateScript [
   &ParticleState()
   &GravityForce(Gravity=&RandomRangeVector(Minimum=<0 0 -35>, Maximum=<1 1 -45>))
   /SetVariables(Particles.Dust SubImageIndex:NiagaraFloat=&FloatFromCurve(FloatCurve=DataInterface() {propertyValues {Curve {keys [<at 0 45>,<at 1 64>]}}})) {isSetParameters=true, module=SetVariables_#6}
   &DynamicMaterialParameters(
    Index 0 Param 1=3
    Index 0 Param 2=0.5
    Index 0 Param 3=1
    Module.Index0Float0Write=true
    Module.Index0Float1Write=true
    Module.Index0Float2Write=true
    Module.Index0Float3Write=true
   )
   &ScaleColor(
    Scale Alpha=&FloatFromCurve(
     FloatCurve=DataInterface() {propertyValues {Curve {
        defaultValue=0.3
        keys [
         <at 0 0 RCIM_Cubic>
         {arriveTangentWeight=0.055, interpMode=RCIM_Cubic, leaveTangentWeight=0.352, tangentMode=RCTM_User, time=0.0947, value=1}
         {arriveTangentWeight=0.5441, interpMode=RCIM_Cubic, leaveTangentWeight=0.6712, tangentMode=RCTM_User, time=0.5, value=1}
         {arriveTangent=-0.156, interpMode=RCIM_Cubic, leaveTangent=-0.156, tangentMode=RCTM_User, time=0.8693, value=0}
        ]
       }}}
     Scale Curve=&FNTimeOfDayScalar(Night Scale=1, Morning Scale=1, Evening Scale=1)
    )
    Scale RGB=&MakeVectorFromLinearColorRGB(Color=&FNTimeOfDayColor(
      Night Color=<rgba 0.3609 0.4715 0.5469 1>
      Morning Color=<rgba 0.651 0.4517 0.4205 1>
      Day Color=<rgba 1 1 1 1>
      Evening Color=<rgba 0.651 0.4517 0.4205 1>
     ))
    ScaleA=true
    ScaleRGB=true
   )
   &ScaleSpriteSize(Uniform Curve Sprite Scale=DataInterface() {propertyValues {Curve {keys [<at 0.0045 0 RCIM_Cubic>, {arriveTangent=1.1, interpMode=RCIM_Cubic, leaveTangent=1.1, time=0.1967, value=0.8}, <at 1.0045 1.1 RCIM_Cubic>]}}})
   &Drag(Drag=3, Use Linear Drag=true, Use Rotational Drag=true)
   &SolveForcesAndVelocity(Clamp Velocity=false, Limit Acceleration=false)
  ]
  rendererClasses [&NiagaraSpriteRendererProperties]
  renderers [{rendererClass=&NiagaraSpriteRendererProperties}]
 }
 FlourPuffs_Pop {
  emitterUpdateScript [
   &EmitterState()
   &SpawnPerFrame(
    Spawn:NiagaraBool=$System.DeathSpawn
    Spawn Count:NiagaraInt32=&RandomRangeInt(Minimum=2, Maximum=4)
    Spawn Group=1
    Use Spawn Probability=false
   ) {module=SpawnPerFrame001}
   &SpawnPerFrame(Spawn:NiagaraBool=$System.DeathSpawn, Spawn Count=3, Use Spawn Probability=false)
   &SpawnBurst_Instantaneous(
    Spawn Count=&RandomRangeInt(Minimum=2, Maximum=4)
    Spawn Group=1
    Spawn Time=5
    Use Spawn Probability=true
    UseLoopCountLimit=false
   ) {enabled=false, module=SpawnBurst_Instantaneous001}
   &SpawnBurst_Instantaneous(Spawn Count=3, Spawn Time=5, Use Spawn Probability=true, UseLoopCountLimit=false) {enabled=false}
  ]
  particleSpawnScript [
   &PartitionParticles() {enabled=false}
   &InitializeParticle(
    Color Maximum=<rgba 1 0.4939 0.7436 0.55>
    Color Minimum=<rgba 1 0.1637 0.5297 0.73>
    Color Mode=::NewEnumerator2
    Lifetime=&MergeHelper_FloatSwitch(
     0=&RandomRangeFloat(Minimum=0.28, Maximum=0.4)
     1=&RandomRangeFloat(Minimum=0.35, Maximum=0.5)
     2=?
     3=?
     Index=&SpawnGroup()
     Num=::NewEnumerator0
    )
    Position Offset=&TransformPosition(Position=$System.DeathPosition, Source Space=::Local, Destination Space=::Local)
    Sprite Rotation Mode=::NewEnumerator2
    Sprite Size Mode=::NewEnumerator1
    Uniform Sprite Size Max=12
    Uniform Sprite Size Min=10
    UsePositionOffset=true
    Write Material Random=true
   )
   &AddVelocity(
    Module.Use Falloff Distance=false
    Override Seed=false
    Velocity Mode=::NewEnumerator1
    Velocity Origin=$Particles.Position
    Velocity Speed=&MergeHelper_FloatSwitch(
     0=12
     1=&RandomRangeFloat(Minimum=45, Maximum=65)
     2=?
     3=?
     Index=&SpawnGroup()
     Num=::NewEnumerator0
    )
   ) {module=AddVelocity001}
   &CameraOffset(Camera Offset Amount=10)
   /SetVariables(Particles.VisibilityTag=&SpawnGroup()) {isSetParameters=true}
   /SetVariables(
    Particles.Sparkle Color:LinearColor=<rgba 5 3.7043 1.51 1>
    Particles.Sparkle Size:Vector2f=&Vector2DFromFloat(Value=&RandomRangeFloat(Minimum=8, Maximum=12))
   ) {isSetParameters=true}
  ]
  particleUpdateScript [
   &ParticleState()
   &GravityForce(Gravity=&MergeHelper_VectorSwitch(
     0=&RandomRangeVector(Minimum=<0 0 -30>, Maximum=<1 1 -20>)
     1=<0 0 20>
     2=?
     3=?
     Index=&SpawnGroup()
     Num=::NewEnumerator0
    ))
   &InheritSourceMovement(
    Applied Position Scale=&MakeCustomVectorFromBool(
     Bool=&SetBoolByFloatComparison(A=&VectorLength(Vector=$Engine.Owner.Velocity), B=500)
     True Vector=<1 1 1>
     False Vector=<0.1 0.1 0>
    )
    ApplyForce=false
    ApplyPosition=true
    ApplyPositionFalloff=false
   )
   /SetVariables(
    Particles.Dust DynamicMaterialParameter:Vector4f=&MakeVector4(X=3, Y=0.5, Z=1, W=0)
    Particles.Dust SubImageIndex:NiagaraFloat=&FloatFromCurve(FloatCurve=DataInterface() {propertyValues {Curve {keys [<at 0 45>,<at 1 64>]}}})
   ) {isSetParameters=true, module=SetVariables_#6}
   &DynamicMaterialParameters(
    Index 0 Param 1=&FloatFromCurve(FloatCurve=DataInterface() {propertyValues {Curve {keys [<at 0 0 RCIM_Cubic>,<at 1.0011 0.2>]}}})
    Index 0 Param 2=&FloatFromCurve(FloatCurve=DataInterface() {propertyValues {Curve {keys [
         {arriveTangent=0.0453, interpMode=RCIM_Cubic, tangentMode=RCTM_Break, time=0, value=-1}
         {arriveTangent=0.0604, interpMode=RCIM_Cubic, leaveTangent=0.0604, tangentMode=RCTM_User, time=1, value=-1}
        ]}}})
    Index 0 Param 3=&FloatFromCurve(FloatCurve=DataInterface() {propertyValues {Curve {keys [{interpMode=RCIM_Cubic, leaveTangent=4.6779, time=0.2851, value=0}]}}})
    Index 1 Param 1=&FloatFromCurve(
     FloatCurve=DataInterface() {propertyValues {Curve {keys [<at 0 0>,<at 0.25 1>]}}}
     Scale Curve=&RandomRangeFloat(Minimum=2, Maximum=1)
    )
    Index 1 Param 2=&FloatFromCurve(
     FloatCurve=DataInterface() {propertyValues {Curve {keys [
         {interpMode=RCIM_Cubic, leaveTangent=3.7422, tangentMode=RCTM_Break, time=0, value=0.2}
         {interpMode=RCIM_Cubic, leaveTangent=0.9066, time=1, value=2}
        ]}}}
     Scale Curve=&RandomRangeFloat(Minimum=0.5, Maximum=1)
    )
    Index 1 Param 3=&FloatFromCurve(FloatCurve=DataInterface() {propertyValues {Curve {keys [
         {interpMode=RCIM_Cubic, leaveTangent=2.8445, tangentMode=RCTM_Break, time=0, value=0.2}
         {interpMode=RCIM_Cubic, leaveTangent=63.2978, time=0.25, value=1}
        ]}}})
    Index 1 Param 4=&FloatFromCurve(
     FloatCurve=DataInterface() {propertyValues {Curve {keys [<at 0 1>,<at 1 0.8>]}}}
     Scale Curve=&RandomRangeFloat(Minimum=0.5, Maximum=1)
    )
    Index 2 Param 1=&FloatFromCurve(FloatCurve=DataInterface() {propertyValues {Curve {keys [{interpMode=RCIM_Cubic, leaveTangent=3.2309, tangentMode=RCTM_Break, time=0.1, value=0}, <at 1 1>]}}})
    Index 2 Param 2=$Particles.MaterialRandom
    Index 2 Param 3=&FloatFromCurve(FloatCurve=DataInterface() {propertyValues {Curve {keys [
         {interpMode=RCIM_Cubic, tangentMode=RCTM_Break, time=0, value=1}
         {interpMode=RCIM_Cubic, leaveTangent=11.1233, tangentMode=RCTM_Break, time=1, value=2}
        ]}}})
    Index 2 Param 4=0.75
    Module.Param0WriteEnabled=true
    Module.Param0WriteEnabled1=true
    Module.Param0WriteEnabled2=true
    Module.Param1WriteEnabled=false
    Module.Param1WriteEnabled1=true
    Module.Param1WriteEnabled2=true
    Module.Param2WriteEnabled=true
    Module.Param2WriteEnabled1=true
    Module.Param2WriteEnabled2=true
    Module.Param3WriteEnabled=true
    Module.Param3WriteEnabled1=true
    Module.Param3WriteEnabled2=true
    Write Parameter Index 1=true
    Write Parameter Index 2=true
   )
   &ScaleColor(
    Scale Alpha=&FloatFromCurve(
     FloatCurve=DataInterface() {propertyValues {Curve {
        defaultValue=0.3
        keys [
         <at 0 0 RCIM_Cubic>
         {arriveTangentWeight=0.055, interpMode=RCIM_Cubic, leaveTangentWeight=0.352, tangentMode=RCTM_User, time=0.0947, value=1}
         {arriveTangentWeight=0.2022, interpMode=RCIM_Cubic, leaveTangentWeight=0.0998, tangentMode=RCTM_User, time=0.6361, value=1}
         {arriveTangent=-4.9566, interpMode=RCIM_Cubic, leaveTangent=-4.9566, tangentMode=RCTM_User, time=1, value=0}
        ]
       }}}
     Scale Curve=&FNTimeOfDayScalar(Night Scale=1, Morning Scale=1, Evening Scale=1)
    )
    Scale RGB=&MakeVectorFromLinearColorRGB(Color=&FNTimeOfDayColor(
      Night Color=<rgba 0.3609 0.4715 0.5469 1>
      Morning Color=<rgba 0.651 0.4517 0.4205 1>
      Day Color=<rgba 1 1 1 1>
      Evening Color=<rgba 0.651 0.4517 0.4205 1>
     ))
    ScaleA=true
    ScaleRGB=true
   )
   &ScaleSpriteSize(Uniform Curve Sprite Scale=DataInterface() {propertyValues {Curve {keys [<at 0 1 RCIM_Cubic>, {arriveTangent=0.3982, interpMode=RCIM_Cubic, leaveTangent=0.3982, time=0.1967, value=1.3}, <at 1.0045 1.4 RCIM_Cubic>]}}})
   &Drag(Drag=4, Use Linear Drag=true, Use Rotational Drag=true)
   &SolveForcesAndVelocity(Clamp Velocity=false, Limit Acceleration=false)
   /SetVariables(
    Particles.Scale Sparkle Color:LinearColor=&MultiplyLinearColor(
     Linear Color=$Particles.Initial.Sparkle Color
     Scale Factor=&MakeVector4_FromLinearColor(LinearColor=&ColorFromCurve(ColorCurve=DataInterface() {propertyValues {
         AlphaCurve {keys [<at 0 1 RCIM_Cubic>,<at 1 0 RCIM_Cubic>]}
         BlueCurve {keys [<at 0 1 RCIM_Cubic>]}
         GreenCurve {keys [<at 0 1 RCIM_Cubic>]}
         RedCurve {keys [<at 0 1 RCIM_Cubic>]}
        }}))
    )
    Particles.Scale Sparkle Size:Vector2f=&Multiply_Vector2D(
     A=$Particles.Initial.Sparkle Size
     B=&Vector2DFromFloat(Value=&FloatFromCurve(FloatCurve=DataInterface() {propertyValues {Curve {keys [
           <at 0 0 RCIM_Cubic>
           {arriveTangentWeight=0.1067, interpMode=RCIM_Cubic, leaveTangentWeight=0.3202, tangentMode=RCTM_User, time=0.05, value=1}
           {arriveTangentWeight=0.3202, interpMode=RCIM_Cubic, leaveTangentWeight=0.1067, tangentMode=RCTM_User, time=0.4905, value=1}
           <at 1 0 RCIM_Cubic>
          ]}}}))
    )
   ) {isSetParameters=true}
  ]
  rendererClasses [&NiagaraSpriteRendererProperties]
  renderers [
   {rendererClass=&NiagaraSpriteRendererProperties}
   {rendererClass=&NiagaraSpriteRendererProperties, rendererIndex=1}
   {rendererClass=&NiagaraSpriteRendererProperties, rendererIndex=2}
  ]
 }
}
summary {
 systemName=NS_Emote_FondCuddles_Biscuits
 emitters [
  {emitterName=Cookies, rendererClasses [&NiagaraMeshRendererProperties]}
  {emitterName=FlourPuffs_Paws, rendererClasses [&NiagaraSpriteRendererProperties]}
  {emitterName=FlourPuffs_Pop, rendererClasses [&NiagaraSpriteRendererProperties]}
 ]
 userVariables [
  User.bIsFrontEndPreview {type=NiagaraBool}
  User.OwnerLocalVelocity {type=Vector3f, description="", defaultValue {struct=Vector3f, value=<0 0 0>}}
  User.SpawnCurve {type=NiagaraFloat, defaultValue=1, description=""}
 ]
}
